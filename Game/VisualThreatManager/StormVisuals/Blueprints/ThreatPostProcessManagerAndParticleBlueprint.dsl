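; Blueprint manager that drives client-side storm visuals: fog post process, camera-attached rain and Epic-quality lens raindrops.
; Threat cloud events cache volumes and play a fade timeline.
; Tick periodically derives a target weight from camera proximity to boxes, or distance outside the safe-zone radius in Athena, then interpolates blend weight, time-of-day fog color and rain spawn rates.
; A timed upward trace stops lens drops under cover.
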(event ThreatVisuals|EventOnThreatCloudsChanged
  (params (ThreatLocationInfo :type Array%20of%20Threat%20Location%20Info%20Structures))
  (bind _returnvalue (Utilities|Array|Length ThreatLocationInfo))
  (bind _returnvalue_2 (!= _returnvalue 0))
  ;[1 1 1] Update cloud/volume information only when changed
  (Utilities|SwitchIsDedicatedServer
    (:Is%20NOT%20Dedicated%20Server
      (Variables|Default|SetIs%20Alive _returnvalue_2)
      (Variables|Default|SetIsAliveJustChanged (Math|Boolean|NotEqual(Boolean) _returnvalue_2 (Variables|Default|GetIsAlivePrevious)))
      (Variables|Default|SetIsAlivePrevious (Variables|Default|GetIs%20Alive))
      ;[1 1 1] This part is redundant with knowledge already held by ThreatVisualsManager
      (Utilities|Array|Clear (Variables|Default|GetArray%20of%20Cloud%20Positions))
      (Utilities|Array|Clear (Variables|Default|GetArray%20of%20Cloud%20Volumes))
      (for (_array_element _array_index) ThreatLocationInfo
        (Utilities|Array|SetArrayElem (Variables|Default|GetArray%20of%20Cloud%20Positions) _array_index (Utilities|Struct|BreakThreatLocationInfo _array_element) true)
        (Utilities|Array|SetArrayElem (Variables|Default|GetArray%20of%20Cloud%20Volumes) _array_index (.ThreatVolume (Utilities|Struct|BreakThreatLocationInfo _array_element)) true))
      ;[1 1 1] Fade Effects in or out
      (if (Variables|Default|GetIsAliveJustChanged)
        (if (Variables|Default|GetIs%20Alive)
          (bind (_direction _newtrack_0) (Timeline|Ramp%20Up%20down%20values%20on%20death Play)
            (:Update
              (Variables|Default|SetRamp%20Up%20And%20Down%20Anim%20Values _newtrack_0)))
          (else
            (Timeline|Ramp%20Up%20down%20values%20on%20death Reverse)))))))

(event ThreatVisuals|EventOnThreatOverrideChanged
  (params (bForceThreatOn :type Boolean))
  (Variables|Default|SetbThreatOverride bForceThreatOn))

(event EventOnWorldReady
  (bind _returnvalue (Game|GetPlayerPawn 0))
  (bind (_asfort_game_state_athena _bsuccess) (Utilities|Casting|CastToFortGameStateAthena (Game|GetGameState)))
  ;[1 1 1] intializeValues
  (Networking|CanExecuteCosmeticEvents
    (:True
      (Variables|Default|SetIsAthenaWorld _bsuccess)
      (Variables|Default|SetIs%20Alive true)
      (CallFunction|StartTraceTimer)
      (Variables|/Script/Engine.PostProcessVolume|SetBlendWeight 0.0 (Variables|Default|GetMaterialpost))
      (Rendering|Material|SetScalarParameterValue "/Game/Packages/Fortress_SharedMaterials/GlobalMaterialParameters/FortniteMaterialParameters.FortniteMaterialParameters" "FogPostProcessOpacity" :!class /Script/Engine.KismetMaterialLibrary)
      (if (Math|Boolean|ANDBoolean (Utilities|IsValid (Game|GetPlayerCameraManager 0) :!class /Script/Engine.KismetSystemLibrary) (Utilities|IsValid (Game|GetPlayerPawn 0) :!class /Script/Engine.KismetSystemLibrary) (Utilities|IsValid _returnvalue :!class /Script/Engine.KismetSystemLibrary)) :!label _label
        (Utilities|FlowControl|Delay 10)
        ;[1 1 1] Camera Lens FX
        (Utilities|FlowControl|Sequence
          (:then_0
            (if (Variables|Default|GetShowRainOnCameraLens)
              (Utilities|EffectsQualityCheck 0
                (:Epic
                  (bind _returnvalue_2 (Camera|AddGenericCameraLensEffect (Game|GetPlayerCameraManager 0) "/Game/VisualThreatManager/Rifts/Blueprints/B_CameraRainDrops_01.B_CameraRainDrops_01_C"))
                  (bind _output_get (Variables|Default|SetCameraLensEffectInterface _returnvalue_2))
                  (Effects|Components|ParticleSystem|SetFloatParameter (CameraLensEffect|GetPrimaryParticleComponent _output_get) "SpawnRate" 0.0))
                (:High
                  _)
                (:Medium
                  _)
                (:Low
                  _))
              (Variables|Default|SetShowRainOnCameraLens false)))
          (:then_1
            ;[1 1 1] Spawn Rain and attach it to the camera
            (if (Variables|Default|GetShowRainEffect)
              (bind _returnvalue_3 (Game|SpawnActorfromClass "/Game/VisualThreatManager/StormVisuals/Blueprints/Threat_RainAttachedToPlayer.Threat_RainAttachedToPlayer_C" (Math|Transform|MakeTransform "0, 0, 0") "Undefined" "OverrideRootScale" 0 0))
              (Transformation|AttachActorToActor _returnvalue_3 _returnvalue "None" "SnapToTarget" "KeepWorld" "KeepWorld" true)
              (bind _output_get_2 (Variables|Default|SetLocalizedParticleSystem (Variables|/Game/VisualThreatManager/StormVisuals/Blueprints/Threat_RainAttachedToPlayer.Threat_RainAttachedToPlayer_C|GetParticleSystem _returnvalue_3)))
              (Effects|Components|ParticleSystem|SetFloatParameter _output_get_2 "SpawnRate" 0.0)
              (Variables|Default|SetIsWorldReady true))))
        (else
          (Utilities|FlowControl|Delay 2)
          (goto _label))))))

(event Custom|StartTraceTimer
  ;[1 1 1] Trace to check and see if you're underneath something (for lens raindrops)
  (Utilities|Time|SetTimerbyEvent (delegate Custom|VFX_RainTracePeriodic) (Variables|Default|GetRain%20Trace%20Timer) true))

(event EventTick
  (params (DeltaSeconds :type Float%20%28single-precision%29))
  (bind _rain_spawn_rate_exponent (Variables|Default|GetRain%20Spawn%20Rate%20Exponent))
  (bind _current_post_process_weight (Variables|PostProcess|GetCurrent%20Post%20Process%20Weight))
  (bind _returnvalue (Math|Float|Power _current_post_process_weight _rain_spawn_rate_exponent))
  (bind _postprocesstargetweight (Variables|PostProcess|GetPostProcessTargetWeight))
  (bind _returnvalue_2 (> _postprocesstargetweight 0.0))
  (bind _returnvalue_3 (Math|Integer|SelectInt 3 1 _returnvalue_2))
  (bind _recalculatetimer (Variables|Default|GetRecalculateTimer))
  (bind _returnvalue_4 (+ _recalculatetimer _returnvalue_3))
  (bind _ramp_up_and_down_anim_values (Variables|Default|GetRamp%20Up%20And%20Down%20Anim%20Values))
  (bind _current_post_process_weight_2 (Variables|PostProcess|GetCurrent%20Post%20Process%20Weight))
  (bind _localizedparticlesystem (Variables|Default|GetLocalizedParticleSystem))
  (bind _postprocesstargetweight_2 (Variables|PostProcess|GetPostProcessTargetWeight))
  (bind _cameralenseffectinterface (Variables|Default|GetCameraLensEffectInterface))
  (if (Variables|Default|GetIsWorldReady)
    (Variables|Default|SetDeltaTime DeltaSeconds)
    (if (or (and (Variables|Default|Getis%20On) (not (Variables|Default|GetbThreatOverride))) (Variables|Default|GetbForceOff))
      (Utilities|FlowControl|Sequence
        (:then_0
          ;[1 1 1] check for clouds
          (if (> _returnvalue_4 (Math|Integer|SelectInt (Variables|Default|GetNumber%20Of%20ticks%20before%20checking%20for%20the%20nearest%20cloud%20again%20-%20when%20alive) (Variables|Default|GetNumber%20Of%20ticks%20before%20checking%20for%20the%20nearest%20cloud%20again%20-%20when%20dead) (Variables|Default|GetIs%20Alive)))
            (Variables|Default|SetRecalculateTimer 0)
            (if (Variables|Default|GetIs%20Alive)
              ;[1 1 1] check all of the volumes for overlap
              (Variables|PostProcess|SetPostProcessTargetWeight 0.0)
              (for _array_element (Variables|Default|GetArray%20of%20Cloud%20Volumes)
                (bind _volumeplayercoveragepercentage (CallFunction|CalculatePlayerPositionNearBox (Utilities|Struct|BreakBox _array_element) (.Max (Utilities|Struct|BreakBox _array_element))))
                (Variables|PostProcess|SetPostProcessTargetWeight (* (Math|Float|Clamp(Float) (+ _volumeplayercoveragepercentage _postprocesstargetweight_2)) _ramp_up_and_down_anim_values))
                ;[1 1 1] Determine Player Sound
                (Variables|Default|SetSoundIsActive (Math|Conversions|ToBoolean(Integer) (Math|Float|Round (* (Math|Float|Clamp(Float) (+ _volumeplayercoveragepercentage _postprocesstargetweight_2)) _ramp_up_and_down_anim_values)))))
              (else
                (Variables|PostProcess|SetPostProcessTargetWeight 0.0)))
            (else
              (Variables|Default|SetRecalculateTimer _returnvalue_4))))
        (:then_1
          ;[1 1 1] Update the post process effects
          (if (!= (Variables|PostProcess|GetCurrent%20Post%20Process%20Weight) (* (Variables|PostProcess|GetPostProcessTargetWeight) (Variables|Default|GetRainMultiplierCurrent))) :!label _label
            (Variables|PostProcess|SetCurrent%20Post%20Process%20Weight (Math|Interpolation|FInterpTo (Variables|PostProcess|GetCurrent%20Post%20Process%20Weight) (Variables|PostProcess|GetPostProcessTargetWeight) (Variables|Default|GetDeltaTime) (Variables|PostProcess|GetPost%20Process%20Fade%20in%20speed)))
            (Variables|/Script/Engine.PostProcessVolume|SetBlendWeight (* _current_post_process_weight_2 (Variables|Default|GetPostProcessWeightModulation)) (Variables|Default|GetMaterialpost))
            (Rendering|Material|SetScalarParameterValue "/Game/Packages/Fortress_SharedMaterials/GlobalMaterialParameters/FortniteMaterialParameters.FortniteMaterialParameters" "FogPostProcessOpacity" _current_post_process_weight_2 :!class /Script/Engine.KismetMaterialLibrary)
            (if (> _current_post_process_weight_2 0.0)
              (bind _color (Utilities|LerpLinearColorsAcrosstheTimesOfDay (Variables|Default|GetThreatFogPostProcessColor_Morning) (Variables|Default|GetThreatFogPostProcessColor_Day) (Variables|Default|GetThreatFogPostProcessColor_Evening) (Variables|Default|GetThreatFogPostProcessColor_Night))
                (:Output
                  (Rendering|Material|SetVectorParameterValue "/Game/Packages/Fortress_SharedMaterials/GlobalMaterialParameters/FortniteMaterialParameters.FortniteMaterialParameters" "FogPostProcessColor" _color :!class /Script/Engine.KismetMaterialLibrary)
                  ;[1 1 1] Spawn rate parameters
                  (if (Variables|Default|GetShowRainEffect)
                    (Utilities|IsValid _localizedparticlesystem
                      (:Is%20Valid
                        (Effects|Components|ParticleSystem|SetFloatParameter _localizedparticlesystem "SpawnRate" _returnvalue))
                      (:Is%20Not%20Valid
                        _)))
                  (if (Variables|Default|GetShowRainOnCameraLens)
                    (LensEffect|IsValidCameraLens _cameralenseffectinterface
                      (:Valid
                        (Effects|Components|ParticleSystem|SetFloatParameter (CameraLensEffect|GetPrimaryParticleComponent _cameralenseffectinterface) "SpawnRate" (* _returnvalue (Variables|Default|GetRainMultiplierCurrent)))))))))))
        (:then_2
          _))
      (else
        (Utilities|FlowControl|Sequence
          (:then_0
            (Variables|PostProcess|SetPostProcessTargetWeight 1.0)
            (goto _label))
          (:then_1
            _))))
    (switch Utilities|FlowControl|Switch|SwitchonESubGame (Fortnite|GetSubGame)
      (:Campaign
        _)
      (:Invalid
        _))
    (Utilities|FlowControl|Delay 1)
    (Utilities|PlayorFadeSound)))

(event Custom|VFX_RainTracePeriodic
  (bind _returnvalue (Game|GetPlayerPawn 0))
  ;[1 1 1] Trace to check and see if you're underneath something (for lens raindrops)
  (if (and (Variables|Default|GetIs%20Alive) (Utilities|IsValid _returnvalue :!class /Script/Engine.KismetSystemLibrary))
    (bind (_outhit _returnvalue_2) (Collision|LineTraceByChannel :Start (Transformation|GetActorLocation _returnvalue) :End (+ (Transformation|GetActorLocation _returnvalue) (Variables|Default|GetRainTraceOffset)) :TraceChannel "TraceTypeQuery10" :bTraceComplex false :DrawDebugType "None" :bIgnoreSelf true :TraceColor "(R=1.000000,G=0.770607,B=0.000000,A=1.000000)" :TraceHitColor "(R=1.000000,G=0.040784,B=0.000000,A=1.000000)" :DrawTime 3))
    (Variables|Default|SetRainMultiplierCurrent (Math|Conversions|ToFloat(Boolean) (not _returnvalue_2)))
    (if (not (Math|Boolean|Equal(Boolean) _returnvalue_2 (Variables|Default|GetTracePrevious)))
      (Variables|Default|SetTracePrevious _returnvalue_2)
      (CallFunction|ForceUpdateLensEffect))))

(event Custom|ForceUpdateLensEffect
  (bind _cameralenseffectinterface (Variables|Default|GetCameraLensEffectInterface))
  ;[1 1 1] Spawn rate parameters
  (LensEffect|IsValidCameraLens _cameralenseffectinterface
    (:Valid
      (Effects|Components|ParticleSystem|SetFloatParameter (CameraLensEffect|GetPrimaryParticleComponent _cameralenseffectinterface) "SpawnRate" (* (Math|Float|Power (Variables|PostProcess|GetCurrent%20Post%20Process%20Weight) (Variables|Default|GetRain%20Spawn%20Rate%20Exponent)) (Variables|Default|GetRainMultiplierCurrent))))))

(fn CalculatePlayerPositionNearBox
  (params (BoxMin :type Vector) (BoxMax :type Vector))
  (bind _returnvalue (Game|GetPlayerCameraManager 0))
  (Utilities|IsValid _returnvalue
    (:Is%20Valid
      (bind _output_get (LocalVariables|SetPlayerCameraLoc (Camera|GetCameraLocation _returnvalue)))
      (if (Variables|Default|GetIsAthenaWorld)
        (bind _output_get_2 (LocalVariables|SetSafeZoneRadius (.x BoxMax)))
        (bind _output_get_3 (LocalVariables|SetDistanceToSafeZone (Math|Vector|VectorLength (* (- (LocalVariables|GetPlayerCameraLoc) BoxMin) "1.000000,1.000000, 0"))))
        (return (Math|Float|MapRangeClamped _output_get_3 _output_get_2 (+ _output_get_2 (Variables|PostProcess|Getpost%20process%20volume%20falloff%20%20around%20the%20threat%20volume)) 0.0 1))
        (else
          (return (Math|Float|Clamp(Float) (* (Math|Float|float*float (- 1 (Math|Float|Clamp(Float) (.x (/ (- BoxMin _output_get) (Variables|PostProcess|Getpost%20process%20volume%20falloff%20%20around%20the%20threat%20volume))))) (- 1 (Math|Float|Clamp(Float) (.y (/ (- BoxMin _output_get) (Variables|PostProcess|Getpost%20process%20volume%20falloff%20%20around%20the%20threat%20volume))))) (- 1 (Math|Float|Clamp(Float) (.z (/ (- BoxMin _output_get) (Variables|PostProcess|Getpost%20process%20volume%20falloff%20%20around%20the%20threat%20volume)))))) (Math|Float|float*float (- 1 (Math|Float|Clamp(Float) (.x (/ (- _output_get BoxMax) (Variables|PostProcess|Getpost%20process%20volume%20falloff%20%20around%20the%20threat%20volume))))) (- 1 (Math|Float|Clamp(Float) (.y (/ (- _output_get BoxMax) (Variables|PostProcess|Getpost%20process%20volume%20falloff%20%20around%20the%20threat%20volume))))) (- 1 (Math|Float|Clamp(Float) (.z (/ (- _output_get BoxMax) (Variables|PostProcess|Getpost%20process%20volume%20falloff%20%20around%20the%20threat%20volume))))))))))))))

(fn SetForceOff
  (params (bForceOff :type Boolean))
  (Variables|Default|SetbForceOff bForceOff))

(fn ConstructionScript)
