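# Material function that fakes a directional highlight, apparently for the storm vortex visuals.
# It transforms the pixel normal from world to view space, dots it with a highlight vector, clamps the result and raises it to an exponent.
# The result is tinted by a highlight color, giving a view-relative rim-like glow without scene lighting.
# All inputs default to the pixel normal or exposed Fake Highlight parameters.
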MaterialFunction /Game/VisualThreatManager/StormVisuals/Materials/MF_Vortex_FakeHighlight_01
graph {
 properties {bExposeToLibrary=true}
 nodes [
  n1 = FunctionInput(Preview=PixelNormalWS()) {bUsePreviewValueAsDefault=true, inputName=Pixel Normal, sortPriority=3}
  n2 = VectorParameter() {group=Fake Highlight, parameterName=Fresnel Highlight Vector, defaultValue=<rgba 0.6 0.2 0.6 0>}
  n3 = Clamp(None=DotProduct(
    A=Transform(None=@n1) {transformSourceType=TRANSFORMSOURCE_World, transformType=TRANSFORM_View}
    B=FunctionInput(Preview=@n2.RGB) {bUsePreviewValueAsDefault=true, inputName=Fresnel Highlight Vector}
   ))
  n4 = FunctionInput(Preview=ScalarParameter() {defaultValue=1.7, group=Fake Highlight, parameterName=Fresnel Highlight Exponent}) {bUsePreviewValueAsDefault=true, inputName=Fresnel Highlight Exponent, inputType=FunctionInput_Scalar, sortPriority=1}
  n5 = VectorParameter() {group=Fake Highlight, parameterName=Fresnel Highlight Color, defaultValue=<rgba 0.1878 0.1425 1.5 0>}
  FunctionOutput(None=Multiply(
    A=Power(Base=@n3, Exp=@n4)
    B=FunctionInput(Preview=@n5.RGB) {bUsePreviewValueAsDefault=true, inputName=Fresnel Highlight Color, sortPriority=2}
   ))
 ]
}
