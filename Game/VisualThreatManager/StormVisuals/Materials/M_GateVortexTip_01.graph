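# Unlit translucent material for a storm gate vortex tip, usable on mesh particles and spline meshes.
# Twist blends two UV sets, then two independently panned noise layers blend two colors, topped with a fake highlight.
# The matching panned normal maps drive world position offset, scaled by V, and perturb the Fresnel normal.
# Opacity combines Fresnel, V gradient, depth fade, particle and vertex alpha, plus an optional timestamp in/out fade.
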Material /Game/VisualThreatManager/StormVisuals/Materials/M_GateVortexTip_01
graph {
 properties {
  bEnableResponsiveAA=true
  bUsedWithMeshParticles=true
  bUsedWithSplineMeshes=true
  blendMode=BLEND_Translucent
  pixelDepthOffsetMode=PDOM_Legacy
  shadingModel=MSM_Unlit
  translucencyLightingMode=TLM_VolumetricPerVertexDirectional
  usedShadingModels=Unlit
 }
 nodes [
  n1 = VectorParameter() {parameterName=Color A, defaultValue=<rgba 0.0344 0.021 0.15 0>}
  n2 = VectorParameter() {parameterName=Color B, defaultValue=<rgba 0.1667 0.12 0.4 0>}
  n3 = DynamicParameter() {paramNames [Twist,None,None,None]}
  n4 = LinearInterpolate(
   A=ScalarParameter() {parameterName=TwistA}
   B=ScalarParameter() {parameterName=TwistB}
   Alpha=ComponentMask(None=TextureCoordinate()) {g=true}
  )
  n5 = MaterialFunctionCall(Float2=LinearInterpolate(
    A=TextureCoordinate()
    B=TextureCoordinate() {coordinateIndex=1}
    Alpha=StaticSwitchParameter(True=@n3.Twist, False=@n4) {defaultValue=true, parameterName=TwistViaDP}
   )) {materialFunction=&BreakOutFloat2Components}
  n6 = AppendVector(A=@n5.R, B=Power(Base=@n5.G, Exp=ScalarParameter() {defaultValue=3, parameterName=V Exponent}))
  n7 = Multiply(A=@n6, B=ScalarParameter() {defaultValue=3, parameterName=U1})
  n8 = Frac(None=Multiply(A=Time(), B=ScalarParameter() {defaultValue=0.111, parameterName=Panner Speed U1}))
  n9 = Multiply(A=@n6, B=ScalarParameter() {defaultValue=0.5861, parameterName=V1})
  n10 = Frac(None=Multiply(A=Time(), B=ScalarParameter() {defaultValue=-0.9786, parameterName=Panner Speed V1}))
  n11 = AppendVector(
   A=ComponentMask(None=Panner(Coordinate=@n7, Time=@n8) {bFractionalPart=true, speedX=1}) {r=true}
   B=ComponentMask(None=Panner(Coordinate=@n9, Time=@n10) {bFractionalPart=true, speedY=1}) {g=true}
  )
  n12 = TextureSample(UVs=@n11) {texture=&2ndInfectedMapLinear}
  n13 = Multiply(A=@n6, B=ScalarParameter() {defaultValue=1, group=Clouds A, parameterName=U2})
  n14 = Frac(None=Multiply(A=Time(), B=ScalarParameter() {defaultValue=0.087, parameterName=Panner Speed U2}))
  n15 = Multiply(A=@n6, B=ScalarParameter() {defaultValue=0.455, parameterName=V2})
  n16 = Frac(None=Multiply(A=Time(), B=ScalarParameter() {defaultValue=-0.6121, parameterName=Panner Speed V2}))
  n17 = AppendVector(
   A=ComponentMask(None=Panner(Coordinate=@n13, Time=@n14) {bFractionalPart=true, speedX=1}) {r=true}
   B=ComponentMask(None=Panner(Coordinate=@n15, Time=@n16) {bFractionalPart=true, speedY=1}) {g=true}
  )
  n18 = TextureSample(UVs=@n17) {texture=&2ndInfectedMapLinear}
  n19 = Power(Base=Divide(A=Add(A=@n12.R, B=@n18.RGB)), Exp=ScalarParameter() {defaultValue=3, parameterName=LERP Color Exponent})
  n20 = Multiply(
   A=Add(
    A=LinearInterpolate(A=@n1.RGB, B=@n2.RGB, Alpha=@n19)
    B=MaterialFunctionCall!(Fresnel Highlight Vector, Fresnel Highlight Exponent, Fresnel Highlight Color, Pixel Normal) {materialFunction=&/MF_Vortex_FakeHighlight_01}.Result
   )
   B=ParticleColor().RGB
  )
  MaterialFunctionCall(Emissive=Multiply(A=@n20, B=ScalarParameter() {defaultValue=1, parameterName=EmissiveStrength})) {materialFunction=&MF_FX_EmissiveAdjustments}
  n21 = ComponentMask(None=TextureCoordinate()) {g=true}
  n22 = VertexNormalWS()
  n23 = Normalize(VectorInput=Add(
    A=TextureSample(UVs=@n11) {samplerType=SAMPLERTYPE_Normal, texture=/Game/Global/Textures/Default/Tileable_Patterns/2ndInfectedMapNormal}.RGB
    B=TextureSample(UVs=@n17) {samplerType=SAMPLERTYPE_Normal, texture=&2ndInfectedMapNormal}.RGB
   ))
  Multiply(A=@n21, B=Multiply(A=Multiply(A=@n22, B=@n23), B=VectorParameter() {parameterName=WPO, defaultValue=<rgba 1.1 1.1 1.1 0>}.RGB))
  Reroute(None=@n20)
  n24 = StaticBoolParameter() {parameterName=Transistion Out Only}
  n25 = VectorParameter() {parameterName=Fresnel Normal Intensity, defaultValue=<rgba 1 1 1 0>}
  n26 = LinearInterpolate(
   A=VertexNormalWS()
   B=Transform(None=Multiply(A=@n23, B=@n25.RGB))
   Alpha=ScalarParameter() {defaultValue=0.1, parameterName=Fresnel Normal Blend}
  )
  n27 = ScalarParameter() {defaultValue=1.5, parameterName=Opacity Gradient Exponent}
  n28 = Multiply(
   A=MaterialFunctionCall(
    Brightness=ScalarParameter() {defaultValue=1, parameterName=Fresnel Brightness}
    Exponent=ScalarParameter() {defaultValue=1.5, parameterName=Fresnel Exponent}
    Vertex / Pixel Normal=Normalize(VectorInput=@n26)
   ) {materialFunction=&MF_Fresnel_FX}.Fresnel Double Sided
   B=Multiply(A=Power(Base=@n21, Exp=@n27), B=ScalarParameter() {defaultValue=2.5, parameterName=Opacity Gradient Brightness})
  )
  n29 = ScalarParameter() {defaultValue=150, parameterName=Depth Fade Distance}
  n30 = Multiply(A=Multiply(A=DepthFade(Opacity=@n28, FadeDistance=@n29), B=ParticleColor().A), B=VertexColor().A)
  StaticSwitchParameter(
   True=Multiply(
    A=MaterialFunctionCall(Transition In Only?=@n24) {materialFunction=&MF_Timestamp_InOut_01}.SmoothStep
    B=@n30
   )
   False=@n30
  ) {parameterName=Use Timestamp Fade?}
 ]
}
