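# Post-process material that tints screen edges with the collection-driven fog color for the storm overlay.
# A slightly off-center radial mask, minus an upward- or downward-panning screen-aligned smooth tile texture, is scaled by FogPostProcessOpacity to set the blend.
# It blends scene color toward FogPostProcessColor, modulated by scene luminance.
# The sky-twist UV distortion and time-of-day color chains feed nothing in the final blend.
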Material /Game/VisualThreatManager/StormVisuals/Materials/M_Screen_Storm_Overlay
graph {
 properties {
  blendMode=BLEND_Translucent
  materialDomain=MD_PostProcess
  opacityMaskClipValue=1
  pixelDepthOffsetMode=PDOM_Legacy
  shadingModel=MSM_Unlit
  usedShadingModels=Unlit
 }
 nodes [
  n1 = Constant2Vector() {g=0.5, r=-0.5}
  n2 = Subtract(A=TextureCoordinate() {uTiling=8, vTiling=8})
  n3 = Constant() {r=0.4}
  n4 = MaterialFunctionCall(Speed=Constant() {r=0.1525}) {materialFunction=&TimeWithSpeedVariable}
  n5 = Multiply(A=ComponentMask(None=MaterialFunctionCall!(
     SineAnimOffset
     Flow Direction=@n1
     Time offset
     EdgeMaskSize
     UVS * size=@n2
     SineAnimSpeed
     Grid Noise Tile Rep
     Rotation amount modulator=@n3
     un frac Time=@n4.Time
    ) {desc=15 fewer instructions on low end, materialFunction=/Game/Packages/Fortress_Sky/Fortnite_SkyTwistFunction}.Rotated UVs) {g=true, r=true}) {constB=0.25}
  Add(A=@n5, B=TextureCoordinate() {uTiling=2, vTiling=2})
  n6 = Panner(Coordinate=Multiply(A=MaterialFunctionCall() {materialFunction=&ScreenAlignedUVs}."X 100%, Y 100%") {constB=2}) {speedY=0.2}
  Add(A=@n6, B=@n5)
  n7 = Constant3Vector() {desc=Morning, constant=<rgba 0.7111 0.3074 1 1>}
  n8 = CollectionParameter() {
   collection=&FortniteMaterialParameters
   parameterName=FortniteActiveTimeOfDay
  }
  n9 = DotProduct(A=@n8, B=Constant4Vector() {constant=<rgba 1 0 0 0>}.RGBA)
  n10 = Constant3Vector() {desc=Day, constant=<rgba 0.64 0.2767 0.9 0>}
  n11 = DotProduct(A=@n8, B=Constant4Vector() {constant=<rgba 0 1 0 0>}.RGBA)
  n12 = DotProduct(A=@n8, B=Constant4Vector() {constant=<rgba 0 0 1 0>}.RGBA)
  n13 = Constant3Vector() {desc=Evening, constant=<rgba 0.7111 0.3074 1 0>}
  n14 = Constant3Vector() {desc=Night, constant=<rgba 0.7111 0.3074 1 0>}
  n15 = DotProduct(A=@n8, B=Constant4Vector() {constant=<rgba 0 0 0 1>}.RGBA)
  Add(A=Add(A=Multiply(A=@n7, B=@n9), B=Multiply(A=@n10, B=@n11)), B=Add(A=Multiply(A=@n12, B=@n13), B=Multiply(A=@n14, B=@n15)))
  n16 = ComponentMask(None=SceneTexture() {sceneTextureId=PPI_PostProcessInput0}.Color) {b=true, g=true, r=true}
  n17 = CollectionParameter() {collection=&FortniteMaterialParameters, parameterName=FogPostProcessColor}
  n18 = Clamp(None=Desaturation(None=@n16))
  n19 = MaterialFunctionCall() {materialFunction=&ScreenAlignedUVs}
  n20 = Subtract(A=ComponentMask(None=@n19."X 100%, Y 100%") {g=true}, B=Constant() {r=0.5})
  n21 = Subtract(A=ComponentMask(None=@n19."X 100%, Y 100%") {r=true}, B=Constant() {r=0.45})
  n22 = Multiply(A=OneMinus(None=Multiply(A=@n20, B=@n20)), B=OneMinus(None=Multiply(A=@n21, B=@n21)))
  n23 = Constant() {desc=30, r=18}
  n24 = Clamp(None=Multiply(A=Power(Base=@n22, Exp=@n23), B=Constant() {desc="4  Lower brings in effect overlay", r=3.3}))
  n25 = Multiply(A=Subtract(A=OneMinus(None=@n24), B=TextureSample(UVs=@n6) {texture=&T_Smooth_tile}.RGB)) {constB=0.21, desc=.18}
  n26 = CollectionParameter() {collection=&FortniteMaterialParameters, parameterName=FogPostProcessOpacity}
  LinearInterpolate(
   A=@n16
   B=Multiply(A=ComponentMask(None=@n17) {b=true, g=true, r=true}, B=LinearInterpolate(Alpha=@n18) {constA=0.5})
   Alpha=Clamp(None=Multiply(A=@n25, B=@n26))
  )
 ]
}
