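# Translucent unlit particle material for storm rain drops on sprite and Niagara particles, with index-of-refraction distortion.
# Emissive samples a fake reflection texture through the drop normal and particle random, shades it by sunlight direction, and tints it by time of day.
# Drop texture, particle alpha and synced distance fade drive opacity and refraction.
# Quality, shading-path and static switches simplify color, normal, fade and refraction for low-end and mobile.
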Material /Game/VisualThreatManager/StormVisuals/Materials/M_Storm_Rain_01
graph {
 properties {
  bEnableResponsiveAA=true
  bUsedWithNiagaraSprites=true
  bUsedWithParticleSprites=true
  blendMode=BLEND_Translucent
  pixelDepthOffsetMode=PDOM_Legacy
  refractionMethod=RM_IndexOfRefraction
  shadingModel=MSM_Unlit
  usedShadingModels=Unlit
 }
 nodes [
  n1 = NamedRerouteUsage()
  n2 = TextureSampleParameter2D() {parameterName=Normal, samplerType=SAMPLERTYPE_Normal, texture=/Game/VisualThreatManager/StormVisuals/Textures/T_Water_drop_N_02}
  n3 = MaterialFunctionCall!(Softness=ScalarParameter() {defaultValue=0.8, parameterName=Softness}, Normal=@n2.RGB) {materialFunction=/Engine/Functions/Engine_MaterialFunctions01/Reflections/ReflectionVectorSoft}
  n4 = ParticleRandom()
  n5 = TextureSample(UVs=ComponentMask(None=Add(A=@n3.Result, B=@n4)) {g=true, r=true}) {texture=/Game/Effects/Fort_Effects/Gameplay/Water/Textures/FX_Reflection_03_Diff}
  n6 = ScalarParameter() {defaultValue=0.01, parameterName=Bias}
  n7 = TextureSampleParameter2D() {parameterName=Normal, samplerType=SAMPLERTYPE_Normal, texture=&T_Water_drop_N_02}
  n8 = CollectionParameter() {
   collection=&FortniteMaterialParameters
   parameterName=SunlightDirectionalVector
  }
  n9 = VectorParameter() {parameterName=FakeSunlightVector, defaultValue=<rgba 0 0 1 0>}
  n10 = DotProduct(
   A=NamedRerouteDeclaration(None=@n7.RGB) {name=Normal, nodeColor=<rgba 1 0 0.0469 1>}
   B=ComponentMask(None=StaticSwitchParameter(True=@n8, False=@n9.RGB) {defaultValue=true, parameterName=UseSunlightDirectionalVector}) {b=true, g=true, r=true}
  )
  n11 = Multiply(A=Multiply(A=Add(A=@n5.RGB, B=@n6), B=ScalarParameter() {defaultValue=4, parameterName=Brightness}), B=@n10)
  n12 = Multiply(
   A=MaterialFunctionCall(
    Color_Morning=VectorParameter() {parameterName=Color Morning, defaultValue=<rgba 1 0.6506 0.3424 0>}.RGB
    Color_Day=VectorParameter() {parameterName=Color Day, defaultValue=<rgba 0.3004 0.4266 0.5 0>}.RGB
    Color_Evening=VectorParameter() {parameterName=Color Evening, defaultValue=<rgba 1.3 0.857 0.533 0>}.RGB
    Color_Night=VectorParameter() {parameterName=Color Night, defaultValue=<rgba 0.27 0.2723 1 0>}.RGB
   ) {materialFunction=&Color_Time_of_Day}.Result
   B=QualitySwitch(Default=@n11, Low=NamedRerouteUsage())
  )
  n13 = StaticSwitchParameter(True=QualitySwitch(Default=ShadingPathSwitch(Default=@n1, Deferred=@n12) {desc=Migrated from FeatureLevelSwitch}, Low=@n1), False=@n12) {group=Performance, parameterName=RemoveTOD_On_Switch_Mobile_Low?}
  n14 = ParticleColor()
  MaterialFunctionCall(Emissive=StaticSwitchParameter(True=Multiply(A=@n13, B=@n14.RGB), False=@n13) {parameterName=UsePaticleColor}) {materialFunction=&MF_FX_EmissiveAdjustments}
  n15 = TextureSample() {texture=/Game/VisualThreatManager/StormVisuals/Textures/T_Water_drop_02}
  n16 = Add(A=@n11, B=Multiply(A=@n10, B=@n15.B))
  n17 = ConstantBiasScale(None=ParticleRandom()) {bias=0.5}
  n18 = QualitySwitch(
   Default=StaticSwitchParameter(True=Multiply(A=@n16, B=@n17), False=@n16) {defaultValue=true, parameterName=IsGPUParticle}
   Low=NamedRerouteUsage()
  )
  n19 = MaterialFunctionCall(
   Position=ParticlePositionWS()
   Fade_Start_Distance=ScalarParameter() {defaultValue=200, parameterName=Start Fade Distance}
   Fade_End_Distance=ScalarParameter() {defaultValue=150, parameterName=End Fade Distance}
  ) {materialFunction=&MF_Distance_SyncedFade_FX}
  Multiply(A=Multiply(A=@n14.A, B=Multiply(A=@n18, B=@n15.A)), B=QualitySwitch(Default=@n19.Result, Low=Constant() {r=1}))
  NamedRerouteDeclaration(None=@n7.B) {name=Normal_B, nodeColor=<rgba 0.5938 0 1 1>}
  n20 = LinearInterpolate(B=LinearInterpolate(B=ScalarParameter() {defaultValue=0.9, parameterName=Refraction}, Alpha=@n15.A) {constA=1}, Alpha=ParticleColor().A) {constA=1}
  Reroute(None=QualitySwitch(Default=LinearInterpolate(B=@n20, Alpha=@n19.Result) {constA=1}, Low=Constant() {r=1}))
  n21 = VectorParameter() {parameterName=Rain_Normal_Intensity, defaultValue=<rgba 0 0 1 1>}
  n22 = ShadingPathSwitch(Default=ShadingPathSwitch(Default=NamedRerouteUsage(), Mobile=@n21.RGB) {desc=Migrated from FeatureLevelSwitch}, Forward=@n21.RGB)
  StaticSwitchParameter(True=@n21.RGB, False=QualitySwitch(Default=@n22, Low=@n21.RGB, High=@n22, Medium=@n22)) {group=Performance, parameterName=Remove_NMap_On_Switch_Mobile_Low?}
 ]
}
