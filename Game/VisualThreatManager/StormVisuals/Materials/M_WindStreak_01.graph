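# Unlit translucent mesh-particle material for wind streaks in storm visuals.
# A panning wisp texture, tinted by particle color, feeds an emissive adjustment function, while vertex normals scaled by thickness form an offset to inflate the mesh.
# Opacity combines particle and vertex alpha, Fresnel and a soft band mask sliding along the V coordinate.
# Dynamic parameters let the particle system drive thickness, mask position, glow and pan speed.
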Material /Game/VisualThreatManager/StormVisuals/Materials/M_WindStreak_01
graph {
 properties {
  bUsedWithMeshParticles=true
  blendMode=BLEND_Translucent
  pixelDepthOffsetMode=PDOM_Legacy
  previewMesh=/Game/VisualThreatManager/StormVisuals/Meshes/Sm_WindStreak_01
  shadingModel=MSM_Unlit
  translucencyLightingMode=TLM_VolumetricPerVertexNonDirectional
  usedShadingModels=Unlit
 }
 nodes [
  n1 = ParticleColor()
  n2 = Time()
  n3 = DynamicParameter() {defaultValue=<rgba 0.5 0.5 10 1>, paramNames [Thickness, PanMask, Emissive Amount, PannerSpeed]}
  n4 = TextureSample(UVs=Panner(Coordinate=TextureCoordinate() {uTiling=0.5, vTiling=0.621}, Time=Multiply(A=@n2, B=@n3.PannerSpeed)) {speedX=-0.23, speedY=-1.8}) {texture=/Game/VisualThreatManager/StormVisuals/Textures/T_Storm_SpoutWisps_01}
  MaterialFunctionCall(Emissive=Multiply(
    A=Multiply(A=@n1.RGB, B=@n4.G)
    B=DynamicParameter() {defaultValue=<rgba 0.5 0.5 10 1>, paramNames [Thickness, PanMask, Emissive Amount, PannerSpeed]}.Emissive Amount
   )) {materialFunction=&MF_FX_EmissiveAdjustments}
  Multiply(
   A=DynamicParameter() {defaultValue=<rgba 0.5 0.5 10 1>, paramNames [Thickness, PanMask, Emissive Amount, PannerSpeed]}.Thickness
   B=VertexNormalWS()
  )
  n5 = VertexColor()
  n6 = ScalarParameter() {defaultValue=1.1, parameterName=Fresnel Brightness}
  n7 = ScalarParameter() {defaultValue=2, parameterName=Fresnel Exponent}
  n8 = Constant()
  n9 = ScalarParameter() {defaultValue=2, parameterName=Mask Length}
  n10 = Multiply(A=DynamicParameter() {defaultValue=<rgba 0.5 0.5 10 1>, paramNames [Thickness, PanMask, Emissive Amount, PannerSpeed]}.PanMask, B=@n9)
  n11 = TextureCoordinate()
  n12 = AppendVector(A=Constant() {r=1}, B=@n9)
  n13 = ComponentMask(None=Subtract(A=AppendVector(A=@n8, B=@n10), B=Multiply(A=@n11, B=@n12))) {g=true}
  n14 = OneMinus(None=@n13)
  Multiply(
   A=Multiply(
    A=Multiply(A=@n1.A, B=@n5.A)
    B=MaterialFunctionCall(Brightness=@n6, Exponent=@n7) {materialFunction=&MF_Fresnel_FX}.Fresnel
   )
   B=Multiply(A=Multiply(A=@n13, B=@n14), B=ScalarParameter() {defaultValue=2, parameterName=Mask Brightness})
  )
 ]
}
